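; Looping gameplay cue for the Jethro CosmosJelly hover jetpack, extending the Jetpack_Hovering notify with backpack VFX and audio.
; OnBurst waits for the Backpack part, then spawns hover Niagara, starts audio and tracks the ascend tag.
; A downward trace places ground dust, while a timer feeds height and fuel into audio parameters.
; OnCeaseRelevant removes hover FX and plays stop or fuel-out sounds; debug prints remain in the ground FX logic.
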(event Custom|OnActiveFX
  (params (Pawn :type Fort%20Player%20Pawn%20Object%20Reference) (Part :type Custom%20Character%20Part%20Object%20Reference) (MeshPart :type Skeletal%20Mesh%20Component%20Object%20Reference) (PartType :type EFortCustomPartType%20Enum :default "Head") (AnyStruct :type Instanced%20Struct%20Structure%20%28by%20ref%29))
  (Utilities|IsValid MeshPart
    (:Is%20Valid
      (bind _output_get (Variables|Default|SetBackpack MeshPart))
      (Variables|Default|GetspawnedHoverSystem
        (:then)
        (:else
          (bind _returnvalue (Niagara|SpawnSystemAttached (Variables|Default|GetNSHoverEffect) _output_get "Root"))
          (Variables|Default|SetspawnedHoverSystem _returnvalue)
          (Utilities|FlowControl|Sequence
            (:then_0
              (Ability|Tasks|WaitGameplayTagAddToActor Pawn "(TagName=\"Granted.Hover.Ascend\")"
                (:Added
                  (Niagara|SetNiagaraVariable(Bool) _returnvalue "User.Ascending" true)
                  (bind _jetpackprimaryaudio (Variables|Audio|GetJetpackPrimaryAudio)
                    (:then
                      (Audio|Parameter|ExecuteTriggerParameter _jetpackprimaryaudio "Play.Ascend"))
                    (:else
                      _))
                  (Ability|Tasks|WaitGameplayTagRemoveFromActor Pawn "(TagName=\"Granted.Hover.Ascend\")"
                    (:Removed
                      (Niagara|SetNiagaraVariable(Bool) _returnvalue "User.Ascending" false)
                      (bind _jetpackprimaryaudio_2 (Variables|Audio|GetJetpackPrimaryAudio)
                        (:then
                          (Audio|Parameter|ExecuteTriggerParameter _jetpackprimaryaudio_2 "Play.StopAscend"))
                        (:else)))))))
            (:then_1
              (Niagara|SetNiagaraVariable(Vector3) _returnvalue "LightSize" (select (Utilities|Equal(SoftObjectReference) (Utilities|ResolveSoftReference (Variables|/Script/Engine.SkinnedMeshComponent|GetSkeletalMesh MeshPart)) (Variables|Default|GetMilo%20Backpack)) (Variables|Default|GetMilo_LightScale) "1.000000,1.000000,1.000000"))
              (HelperMethods|SetJetpackAudioEnabled true)
              (Rendering|Material|SetScalarParameterValueonMaterials MeshPart "FullBlink" 1.0)
              (Rendering|Material|SetScalarParameterValueonMaterials MeshPart "Refueling" 0.0)
              (Variables|Default|SetPawn Pawn))))))))

(event Custom|StoreHeadPart
  (params (Pawn :type Fort%20Player%20Pawn%20Object%20Reference) (Part :type Custom%20Character%20Part%20Object%20Reference) (MeshPart :type Skeletal%20Mesh%20Component%20Object%20Reference) (PartType :type EFortCustomPartType%20Enum :default "Head") (AnyStruct :type Instanced%20Struct%20Structure%20%28by%20ref%29))
  (Variables|Default|SetHeadPart MeshPart))

(event Custom|On%20Pawn%20Landed
  (params (Hit :type Hit%20Result%20Structure%20%28by%20ref%29))
  ;[1 1 1] Reset "First-Jump" flag once we've landed
  (Variables|Sounds|SetbIsFirstJump true)
  (bind _playerpawn (Variables|GameplayCueNotify_Jetpack_Hovering|GetPlayerPawn)
    (:then
      (EventDispatchers|UnbindEventfromOnPawnLanded _playerpawn (delegate Custom|On%20Pawn%20Landed)))
    (:else)))

(event Custom|ResetFX
  (bind _groundinteract (Variables|Default|GetGroundInteract))
  (bind _groundinteract_2 (Variables|Default|GetGroundInteract))
  ;[1 1 1] Ground FX Control
  (Utilities|FlowControl|Sequence
    (:then_0
      (Utilities|FlowControl|DoOnce false
        (:Completed
          (Utilities|IsValid _groundinteract_2
            (:Is%20Valid
              (Components|Activation|Deactivate _groundinteract_2)
              (Development|PrintString 0 "Deactivate 1"))
            (:Is%20Not%20Valid
              _))
          (bind _headpart (Variables|Default|GetHeadPart)
            (:then
              (bind _returnvalue (Niagara|SpawnSystemAttached "/JethroWeaponGameplay/Assets/FX/Jetpack/GroundDust/NS_Jethro_Jelly_GroundDust.NS_Jethro_Jelly_GroundDust" _headpart "GroundInteract"))
              (bind _output_get (Variables|Default|SetGroundInteract _returnvalue))
              (Development|PrintString 0 (Utilities|GetDisplayName _output_get)))
            (:else)))))
    (:then_1
      (Utilities|IsValid _groundinteract
        (:Is%20Valid
          (Components|Activation|Deactivate _groundinteract)
          (Development|PrintString 0 "deactivate 2"))))))

(event GameplayCueNotify|EventHandleGameplayCue
  (params (MyTarget :type Actor%20Object%20Reference) (EventType :type EGameplayCueEvent%20Enum :default "OnActive") (Parameters :type Gameplay%20Cue%20Parameters%20Structure%20%28by%20ref%29))
  (bind _outoffuelaudiocomp (Variables|GameplayCueNotify_Jetpack_Hovering|GetOutOfFuelAudioComp))
  ;[1 1 1] Fix for loop getting stuck
  (Utilities|IsValid MyTarget
    (:Is%20Valid
      (if (Utilities|Enum|Equal(Enum) EventType "Removed")
        (Utilities|IsValid _outoffuelaudiocomp
          (:Is%20Valid
            (if (Audio|Components|Audio|IsPlaying _outoffuelaudiocomp)
              (Audio|Components|Audio|Stop _outoffuelaudiocomp))))))))

(event EventTryToPlayGroundEffects
  (bind (_out_ground_hit_bblockinghit _out_ground_hit_binitialoverlap _out_ground_hit_time _out_ground_hit_distance _out_ground_hit_location _out_ground_hit_impactpoint _out_ground_hit_normal _out_ground_hit_impactnormal _out_ground_hit_physmat _out_ground_hit_hitactor _out_ground_hit_hitcomponent _out_ground_hit_hitbonename _out_ground_hit_bonename _out_ground_hit_hititem _out_ground_hit_elementindex _out_ground_hit_faceindex _out_ground_hit_tracestart) (Collision|BreakHitResult (Variables|GetOut%20Ground%20Hit_bBlockingHit)))
  (bind _groundinteract (Variables|Default|GetGroundInteract))
  (bind _groundinteract_2 (Variables|Default|GetGroundInteract))
  (bind _groundinteract_3 (Variables|Default|GetGroundInteract))
  (bind _groundinteract_4 (Variables|Default|GetGroundInteract))
  ;[1 1 1] Fix for loop getting stuck
  (Development|PrintString 0 "Try play Ground FX")
  (bind _backpack (Variables|Default|GetBackpack)
    (:then
      ;[1 1 1] Ground FX Control
      (bind (_outhit _returnvalue) (Collision|LineTraceForObjects :Start (Transformation|GetSocketLocation _backpack "GroundInteract") :End (+ (Transformation|GetSocketLocation _backpack "GroundInteract") "0, 0,-650.000000") :ObjectTypes (Utilities|Array|MakeArray "ObjectTypeQuery1") :bTraceComplex false :DrawDebugType "None" :bIgnoreSelf true :TraceColor "(R=1.000000,G=0.000000,B=0.000000,A=1.000000)" :TraceHitColor "(R=0.000000,G=1.000000,B=0.000000,A=1.000000)" :DrawTime 1.0))
      (Variables|Default|SetOut%20Ground%20Hit (select _returnvalue _outhit (Variables|Default|GetOut%20Ground%20Hit)))
      (if _returnvalue
        (Utilities|FlowControl|Sequence
          (:then_0
            (Utilities|FlowControl|DoOnce false
              (:Completed
                (Utilities|IsValid _groundinteract_2
                  (:Is%20Valid
                    (Components|Activation|Deactivate _groundinteract_2)
                    (Development|PrintString 0 "Deactivate 1"))
                  (:Is%20Not%20Valid
                    _))
                (bind _headpart (Variables|Default|GetHeadPart)
                  (:then
                    (bind _returnvalue_2 (Niagara|SpawnSystemAttached "/JethroWeaponGameplay/Assets/FX/Jetpack/GroundDust/NS_Jethro_Jelly_GroundDust.NS_Jethro_Jelly_GroundDust" _headpart "GroundInteract"))
                    (bind _output_get (Variables|Default|SetGroundInteract _returnvalue_2))
                    (Development|PrintString 0 (Utilities|GetDisplayName _output_get)))
                  (:else)))))
          (:then_1
            (Utilities|IsValid _groundinteract_4
              (:Is%20Valid
                (Transformation|SetWorldLocationAndRotation _groundinteract_4 _out_ground_hit_location (Math|Rotator|MakeRotfromZ _out_ground_hit_impactnormal) false false)
                (Effects|Components|ParticleSystem|SetFloatParameter (Variables|Default|GetGroundInteract) "Height" (/ (Math|Vector|VectorLength (- _out_ground_hit_location _out_ground_hit_tracestart)) 6.5e+02)))))
          (:then_2
            (Utilities|IsValid _groundinteract_3
              (:Is%20Valid
                (Components|Activation|Activate _groundinteract_3 true)))))
        (else
          (Utilities|IsValid _groundinteract
            (:Is%20Valid
              (Components|Activation|Deactivate _groundinteract)
              (Development|PrintString 0 "deactivate 2"))))))
    (:else)))

(fn ApplyDistanceToAudio
  (bind _distance (HelperMethods|CalculateDistancetoObjects))
  (bind _jetpackprimaryaudio (Variables|Audio|GetJetpackPrimaryAudio)
    (:then
      (Audio|Parameter|SetFloatParameter _jetpackprimaryaudio "NormalizedDistanceToObjects" (/ _distance (Variables|Audio|GetMaxTraceDistance)) :!class /Script/AudioExtensions.AudioParameterControllerInterface)
      (CallFunction|GetFuelforJetpackAudio))
    (:else)))

(fn CalculateDistanceToObjects
  (bind _backpack (Variables|Default|GetBackpack)
    (:then
      (bind _outhit (Collision|LineTraceByChannel 0 (- (Transformation|GetWorldLocation _backpack) "0.000000,0.000000,100.000000") (Utilities|Operators|Subtract (- (Transformation|GetWorldLocation _backpack) "0.000000,0.000000,100.000000") (Math|Vector|MakeVector 0.0 0.0 (Variables|Audio|GetMaxTraceDistance))) "TraceTypeQuery1" false (Utilities|Array|MakeArray (Variables|Default|GetPawn))))
      (return (.Distance (Collision|BreakHitResult _outhit))))
    (:else)))

(fn GetFuelForJetpackAudio
  (bind _pawn (Variables|Default|GetPawn)
    (:then
      (if (Ability|Attribute|GetFloatAttributefromAbilitySystemComponent (Ability|GetAbilitySystemComponent _pawn) "(AttributeName=\"Fuel\",Attribute=/Script/FortniteGame.FortJumpBoostSet:Fuel,AttributeOwner=\"/Script/CoreUObject.Class'/Script/FortniteGame.FortJumpBoostSet'\")")
        (bind _jetpackprimaryaudio (Variables|Audio|GetJetpackPrimaryAudio)
          (:then
            (Audio|Parameter|SetFloatParameter _jetpackprimaryaudio "Fuel.Normalized" (/ (.ReturnValue (Ability|Attribute|GetFloatAttributefromAbilitySystemComponent (Ability|GetAbilitySystemComponent _pawn) "(AttributeName=\"Fuel\",Attribute=/Script/FortniteGame.FortJumpBoostSet:Fuel,AttributeOwner=\"/Script/CoreUObject.Class'/Script/FortniteGame.FortJumpBoostSet'\")")) 1e+02) :!class /Script/AudioExtensions.AudioParameterControllerInterface)
            (return (- 1.0 (/ (.ReturnValue (Ability|Attribute|GetFloatAttributefromAbilitySystemComponent (Ability|GetAbilitySystemComponent _pawn) "(AttributeName=\"Fuel\",Attribute=/Script/FortniteGame.FortJumpBoostSet:Fuel,AttributeOwner=\"/Script/CoreUObject.Class'/Script/FortniteGame.FortJumpBoostSet'\")")) 1e+02))))
          (:else))))
    (:else)))

(fn OnBurst
  (params (MyTarget :type Actor%20Object%20Reference) (Parameters :type Gameplay%20Cue%20Parameters%20Structure%20%28by%20ref%29))
  (bind _playerpawn (Variables|GameplayCueNotify_Jetpack_Hovering|GetPlayerPawn))
  (bind _returnvalue (|CallParentFunction MyTarget Parameters))
  (if _returnvalue
    (Default|QueueWorkwhenSpecificCosmeticPartApplied _playerpawn (EventDispatchers|CreateEvent :!event OnActiveFX) false "Backpack")
    (Default|QueueWorkwhenSpecificCosmeticPartApplied _playerpawn (EventDispatchers|CreateEvent :!event StoreHeadPart) false "Backpack")
    (return true)
    (else
      (return false))))

(fn OnCeaseRelevant
  (params (MyTarget :type Actor%20Object%20Reference) (Parameters :type Gameplay%20Cue%20Parameters%20Structure%20%28by%20ref%29))
  (bind _groundinteract (Variables|Default|GetGroundInteract))
  (Class|FortQuestBlueprintLibrary|GetFortPawnFromObject MyTarget
    (:Valid
      (bind _pawn (Variables|Default|GetPawn)
        (:then
          (EventDispatchers|BindEventtoOnPawnLanded _pawn (EventDispatchers|CreateEvent :!event On%20Pawn%20Landed)))
        (:else)))
    (:NotValid
      _))
  (Utilities|FlowControl|Sequence
    (:then_0
      (CallFunction|ResetFX)
      (Utilities|IsValid _groundinteract
        (:Is%20Valid
          (Components|Activation|Deactivate _groundinteract))))
    (:then_1
      (bind _spawnedhoversystem (Variables|Default|GetspawnedHoverSystem)
        (:then
          (Components|Activation|Deactivate _spawnedhoversystem)
          (Components|DestroyComponent _spawnedhoversystem 0))
        (:else)))
    (:then_2
      (HelperMethods|SetJetpackAudioEnabled false))
    (:then_3
      (|CallParentFunction (Variables|GetMyTarget) (Variables|GetParameters))
      (return true))))

(fn SetJetpackAudioEnabled
  (params (Enabled :type Boolean))
  (bind _heightaudiomodulatortimer (Variables|Audio|GetHeightAudioModulatorTimer))
  (bind _output_get (Variables|Sounds|SetbJetpackAudioEnabled Enabled))
  ;[0.6574 0.1718 0.8594 0.5] C++ deals with deactivate
  (Actor|Tick|SetActorTickEnabled _output_get)
  (Variables|GameplayCueNotify_Jetpack_Hovering|GetPlayerPawn
    (:then
      (if (Variables|Sounds|GetbJetpackAudioEnabled)
        (EventDispatchers|BindEventtoOnPawnLanded (Variables|GameplayCueNotify_Jetpack_Hovering|GetPlayerPawn) (EventDispatchers|CreateEvent :!event On%20Pawn%20Landed))
        (Variables|Audio|GetJetpackPrimaryAudio
          (:then
            _)
          (:else
            (bind _returnvalue (Audio|SpawnSoundAttached "/CosmosGameplay/CosmosJelly/Sounds/IP_MSS_Jethro_CosmosJelly_Jetpack.IP_MSS_Jethro_CosmosJelly_Jetpack" (Variables|Default|GetBackpack)))
            (bind _output_get_2 (Variables|Audio|SetJetpackPrimaryAudio _returnvalue))))
        (Audio|Parameter|ExecuteTriggerParameter _output_get_2 "Play.StartEngine")
        (Audio|SpawnSoundAttached "/CosmosGameplay/CosmosJelly/Sounds/IP_Cosmos_Jetpack_Start_Cue.IP_Cosmos_Jetpack_Start_Cue" (Variables|Default|GetBackpack))
        (Variables|Sounds|SetbIsFirstJump false)
        (if (not (Utilities|Time|IsTimerActivebyHandle 0 _heightaudiomodulatortimer))
          (Utilities|Time|ClearandInvalidateTimerbyHandle 0 _heightaudiomodulatortimer))
        (bind _returnvalue_2 (Utilities|Time|SetTimerbyFunctionName 0 "ApplyDistanceToAudio" 0.1 true true))
        (Variables|Audio|SetHeightAudioModulatorTimer _returnvalue_2)
        (else
          _)))
    (:else
      _))
  (bind _remainingfuel0_to1 (CallFunction|GetFuelforJetpackAudio))
  (Utilities|FlowControl|Sequence
    (:then_0
      (Audio|SpawnSoundAttached (select (<= _remainingfuel0_to1 0) "/CosmosGameplay/CosmosJelly/Sounds/IP_Cosmos_Jetpack_FuelOut_Cue.IP_Cosmos_Jetpack_FuelOut_Cue" "/CosmosGameplay/CosmosJelly/Sounds/IP_Cosmos_Jetpack_Stop_Cue.IP_Cosmos_Jetpack_Stop_Cue") (Variables|Default|GetBackpack)))
    (:then_1
      (bind _outoffuelaudiocomp (Variables|GameplayCueNotify_Jetpack_Hovering|GetOutOfFuelAudioComp)
        (:then
          (Audio|Components|Audio|Stop _outoffuelaudiocomp))
        (:else)))
    (:then_2
      (bind _idleaudiocomp (Variables|GameplayCueNotify_Jetpack_Hovering|GetIdleAudioComp)
        (:then
          (Audio|Components|Audio|Stop _idleaudiocomp))
        (:else)))
    (:then_3
      (bind _accelaudiocomp (Variables|GameplayCueNotify_Jetpack_Hovering|GetAccelAudioComp)
        (:then
          (Audio|Components|Audio|Stop _accelaudiocomp))
        (:else)))
    (:then_4
      (bind _jetpackprimaryaudio (Variables|Audio|GetJetpackPrimaryAudio)
        (:then
          (Audio|Parameter|ExecuteTriggerParameter _jetpackprimaryaudio "Play.StopEngine"))
        (:else)))))

(fn ConstructionScript)
